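# Physics asset for the M_MED_AnodizedMetal backpack cosmetic, providing simple collision bodies and joints for its dangling masks, rope tails and ribbons.
# Capsules cover the pack and most dyn_ bones, with a sphere on the rope base.
# The rope base joins the pack with free swing and twist, acting as a hub for both twist-locked masks and the four upper and lower rope tails.
# The two ribbons constrain directly to the pack.
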PhysicsAsset /BRCosmetics/Accessories/FORT_Backpacks/Backpack_M_MED_AnodizedMetal/Meshes/M_MED_AnodizedMetal_Pack_FE_Physics
bodies {
 dyn_lower_ropetail_01_l [{
   shapeType=Capsule
   length=1.4
   radius=1
   shapeName=dyn_lower_ropetail_01_l_capsule
   center=<2.6225 -0.2425 0.0835>
   rotation=<rot 0 -95.2838 -78.018>
  }]
 dyn_lower_ropetail_01_r [{
   shapeType=Capsule
   length=0.7
   radius=1
   shapeName=dyn_lower_ropetail_01_r_capsule
   center=<3.5379 0.1938 -0.1227>
   rotation=<rot -0.0001 -86.8649 -78.9905>
  }]
 dyn_mask_l [
  {
   shapeType=Capsule
   length=1.7112
   radius=2.4333
   shapeName=dyn_mask_r_capsule
   center=<3.8704 0.3161 -2.1936>
   rotation=<rot -5.2795 -84.7718 -40.7586>
  }
  {shapeType=Capsule, length=2.5, radius=2, shapeName=dyn_mask_r_capsule, center=<1.0494 -0.0697 -0.7859>, rotation=<rot 57.2271 0 -90>}
 ]
 dyn_mask_r [
  {shapeType=Capsule, length=3.2112, radius=1.9333, shapeName=dyn_mask_r_capsule, center=<1.254 -0.0786 -1.1208>, rotation=<rot 0 0 -90>}
  {shapeType=Capsule, length=0, radius=3.5, shapeName=dyn_mask_r_capsule, center=<1.2098 -0.0532 2.7093>, rotation=<rot 90 -172.875 83.0888>}
 ]
 dyn_robbon_l [{shapeType=Capsule, length=0.6, radius=0.9, shapeName=dyn_robbon_r_capsule, center=<1.3583 0.4136 0.4963>, rotation=<rot 90 0 0>}]
 dyn_robbon_r [{shapeType=Capsule, length=0.6, radius=0.9, shapeName=dyn_robbon_r_capsule, center=<1.5671 0 0.4963>, rotation=<rot 90 0 0>}]
 dyn_ropebase [{shapeType=Sphere, radius=2, shapeName=dyn_ropebase_sphere, center=<6.2969 -0.1566 0.227>}]
 dyn_upper_ropetail_01_l [{
   shapeType=Capsule
   length=1.5704
   radius=1.1235
   shapeName=dyn_upper_ropetail_01_l_capsule
   center=<1.8378 0.0654 -0.0404>
   rotation=<rot 0 -86.868 -86.6732>
  }]
 dyn_upper_ropetail_01_r [{
   shapeType=Capsule
   length=1.3536
   radius=1.1497
   shapeName=dyn_upper_ropetail_01_r_capsule
   center=<1.3357 -0.0006 0.0785>
   rotation=<rot 0 87.1226 -85.9965>
  }]
 pack [
  {shapeType=Capsule, length=0, radius=8.9, shapeName=pack_capsule, center=<2.0496 -11.8311 -10.3528>, rotation=<rot 28.7331 -69.4021 -3.1365>}
  {shapeType=Capsule, length=4, radius=2, shapeName=pack_capsule, center=<-2.8038 -2.8987 9.4732>, rotation=<rot 28.7331 -69.4021 -3.1365>}
 ]
}
constraints [
 {bone1Name=dyn_mask_l, bone2Name=dyn_ropebase, swing1LimitDegrees=10, swing2LimitDegrees=10, twistLimitDegrees=2, twistMotion=Locked}
 {bone1Name=dyn_mask_r, bone2Name=dyn_ropebase, swing1LimitDegrees=2, swing2LimitDegrees=10, twistLimitDegrees=2, twistMotion=Locked}
 {bone1Name=dyn_upper_ropetail_01_l, bone2Name=dyn_ropebase, swing1LimitDegrees=20, swing2LimitDegrees=20, twistLimitDegrees=0}
 {bone1Name=dyn_upper_ropetail_01_r, bone2Name=dyn_ropebase, swing1LimitDegrees=20, swing2LimitDegrees=20, twistLimitDegrees=0}
 {
  bone1Name=dyn_ropebase
  bone2Name=pack
  swing1LimitDegrees=45
  swing1Motion=Free
  swing2LimitDegrees=45
  swing2Motion=Free
  twistLimitDegrees=45
  twistMotion=Free
 }
 {bone1Name=dyn_lower_ropetail_01_l, bone2Name=dyn_ropebase, swing1LimitDegrees=70, swing2LimitDegrees=60, twistLimitDegrees=5}
 {bone1Name=dyn_lower_ropetail_01_r, bone2Name=dyn_ropebase, swing1LimitDegrees=40, swing2LimitDegrees=60, twistLimitDegrees=5}
 {bone1Name=dyn_robbon_r, bone2Name=pack, swing1LimitDegrees=60, swing2LimitDegrees=63, twistLimitDegrees=0}
 {bone1Name=dyn_robbon_l, bone2Name=pack, swing1LimitDegrees=50, swing2LimitDegrees=60, twistLimitDegrees=0}
]
