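# Anim Blueprint for the AnodizedMetal backpack cosmetic, adding physics motion to its rope and dynamic bones.
# A cached, slightly offset pack pose feeds two AnimDynamics-to-RigidBody chains, selected by bIsInFrontEnd.
# The front-end chain uses world-space simulation, its own physics asset and rolled upper rope tails.
# The other chain simulates relative to root, with gravity set each update by a deprecated override function.
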AnimBlueprint /BRCosmetics/Accessories/FORT_Backpacks/Backpack_M_MED_AnodizedMetal/Meshes/M_MED_AnodizedMetal_Pack_AnimBP
target_skeleton=&/M_MED_AnodizedMetal_Pack_Skeleton
graphs [
 AnimGraph {nodes [
   n1 = Anim Dynamics - Bone: dyn_ropebase(ComponentPose=Local To Component(LocalPose=Use cached pose 'NotFrontEnd'()), Alpha=0.6, ExternalForce="0,0,0") {
    angularDampingOverride=0.95
    angularSpringConstant=500
    bAngularSpring=true
    bLinearSpring=true
    bOverrideAngularDamping=true
    bOverrideLinearDamping=true
    bUseGravityOverride=true
    linearDampingOverride=0.95
    linearSpringConstant=500
    numSolverIterationsPostUpdate=4
    numSolverIterationsPreUpdate=16
    simulationSpace=World
    alphaBoolBlend {blendOption=Linear}
    boundBone {boneName=dyn_ropebase}
    componentLinearAccScale=<0.7 0.7 0.7>
    componentLinearVelScale=<0.7 0.7 0.7>
    gravityOverride {z=-980}
    physicsBodyDefinitions [{
      boundBone {boneName=dyn_ropebase}
      boxExtents=<20 20 20>
      constraintSetup {angularLimitsMax=<0 0 2.5>, angularLimitsMin=<0 -8 -2.5>, angularTarget=<0 0 0>}
      localJointOffset=<1 0 0>
     }]
   }
   n2 = RigidBody!(ComponentPose=@n1, Alpha=0.6, OverrideWorldGravity:Vector="0,0,-980", ExternalForce) -> (Pose) {
    bOverrideWorldGravity=true
    overridePhysicsAsset=&/M_MED_AnodizedMetal_Pack_FE_Physics
    simulationSpace=WorldSpace
    alphaBoolBlend {blendOption=Linear}
    baseBoneRef {boneName=root}
    componentAppliedLinearAccClamp=<3000 3000 3000>
    componentLinearAccScale {y=1, z=1}
    componentLinearVelScale {y=1, z=1}
    overrideWorldGravity {z=-980}
    simSpaceSettings {maxAngularAcceleration=1000, maxAngularVelocity=3, maxLinearAcceleration=700, maxLinearVelocity=500, velocityScaleZ=0, worldAlpha=1}
   }
   n3 = "Transform (Modify) Bone - Bone: dyn_upper_ropetail_01_r"(ComponentPose=@n2, Alpha=1, Translation="0,0,0", Rotation="0,0,-60", Scale="1,1,1") {rotationMode=BMM_Additive, alphaBoolBlend {blendOption=Linear}, boneToModify {boneName=dyn_upper_ropetail_01_r}, rotation {roll=-60}}
   n4 = "Transform (Modify) Bone - Bone: dyn_upper_ropetail_01_l"(ComponentPose=@n3, Alpha=1, Translation="0,0,0", Rotation="0,0,-60", Scale="1,1,1") {rotationMode=BMM_Additive, alphaBoolBlend {blendOption=Linear}, boneToModify {boneName=dyn_upper_ropetail_01_l}, rotation {roll=-60}}
   n5 = Anim Dynamics - Bone: dyn_ropebase(ComponentPose=Local To Component(LocalPose=Use cached pose 'NotFrontEnd'()), Alpha=1, ExternalForce="0,0,0") {
    angularDampingOverride=0.95
    angularSpringConstant=500
    bAngularSpring=true
    bLinearSpring=true
    bOverrideAngularDamping=true
    bOverrideLinearDamping=true
    bUseGravityOverride=true
    linearDampingOverride=0.95
    linearSpringConstant=500
    numSolverIterationsPostUpdate=4
    numSolverIterationsPreUpdate=16
    simulationSpace=RootRelative
    alphaBoolBlend {blendOption=Linear}
    boundBone {boneName=dyn_ropebase}
    componentLinearAccScale=<0.7 0.7 0.7>
    componentLinearVelScale=<0.7 0.7 0.7>
    gravityOverride {z=-980}
    physicsBodyDefinitions [{
      boundBone {boneName=dyn_ropebase}
      boxExtents=<20 20 20>
      constraintSetup {angularLimitsMax=<0 0 2.5>, angularLimitsMin=<0 -8 -2.5>, angularTarget=<0 0 0>}
      localJointOffset=<1 0 0>
     }]
   }
   n6 = RigidBody(ComponentPose=@n5, OverrideWorldGravity=Get Gravity Override()) {
    bClampLinearTranslationLimitToRefPose=true
    bOverrideWorldGravity=true
    overridePhysicsAsset=&/M_MED_AnodizedMetal_Pack_Physics
    simulationSpace=BaseBoneSpace
    alphaBoolBlend {blendOption=Linear}
    baseBoneRef {boneName=root}
    componentAppliedLinearAccClamp=<3000 3000 3000>
    componentLinearAccScale=<0.4 1 1>
    componentLinearVelScale=<0.4 1 1>
    overrideWorldGravity {z=-980}
    simSpaceSettings {maxAngularAcceleration=1000, maxAngularVelocity=3, maxLinearAcceleration=700, maxLinearVelocity=500, velocityScaleZ=0, worldAlpha=1}
   }
   Output Pose(Result=Blend Poses by bool(BlendPose_0=Component To Local(ComponentPose=@n4), BlendPose_1=Component To Local(ComponentPose=@n6), bActiveValue=Get bIsInFrontEnd()) {blendType=Linear}) {layerGroup=None}
   n7 = "Transform (Modify) Bone - Bone: pack"(ComponentPose=Local To Component(LocalPose=Input Pose()), Translation="0,1,0", Rotation="0,0,0") {
    translationMode=BMM_Additive
    translationSpace=BCS_BoneSpace
    alphaBoolBlend {blendOption=Linear}
    boneToModify {boneName=pack}
    translation {y=1}
   }
   Save cached pose 'NotFrontEnd'(Pose=Component To Local(ComponentPose=@n7)) {cacheName=NotFrontEnd}
  ]}
 EventGraph {nodes [
   n1 = Event BlueprintUpdateAnimation()
   n2 = Gravity Override State Function DEPRECATED(
    execute=[@n1.then, @n3.else]
    CharacterPartAnimInstance=Self-Reference()
    GameTimeInSeconds=GetGameTimeInSeconds()
    GravityOverrideParams=Get Gravity Override Params()
   )
   > Set Gravity Override(Gravity Override=@n2)
   n3 = Branch(Condition=Get bIsInFrontEnd())
   > n4 = Gravity Override State Function DEPRECATED(CharacterPartAnimInstance=Self-Reference(), GameTimeInSeconds=GetGameTimeInSeconds(), GravityOverrideParams=Get Gravity Override Params())
   > Set Gravity Override(Gravity Override=@n4)
   TryGetPawnOwner()
  ]}
]
variables [
 Gravity Override {type=Vector}
 Gravity Override Params {type=GravityOverrideParamsStruct, default {front_End {gravityOverride=<0 0 -3000>}}}
]
